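; Galileo prototype Blueprint that appears to pre-buffer the in-match live stream video by turning on hidden picture-in-picture playback on clients.
; On BeginPlay, client mesh network nodes wait a random delay of up to three minutes, likely staggering stream starts.
; It then saves the live stream runtime options, forces the enable, show-by-default and hidden-but-enabled flags on, and randomizes the PiP counter.
; EndPlay writes the saved values back on clients.
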(event EventBeginPlay
  (bind _returnvalue (Options|GetRuntimeOptions))
  (bind _shouldshowlivestreampipinmatchcounter (Variables|/Script/FortniteGame.FortRuntimeOptions|GetShouldShowLiveStreamPiPInMatchCounter _returnvalue))
  ;[1 1 1] Begin Play
  (switch Utilities|FlowControl|Switch|SwitchonEMeshNetworkNodeType (MeshNetworkSubsystem|GetMeshNetworkNodeType (GameInstanceSubsystems|GetMeshNetworkSubsystem))
    (:Client
      (bind _output_get (Variables|Default|Settmp (Math|Random|RandomFloatinRange 0.0 1.8e+02)))
      (Development|PrintString (Utilities|String|Append "Start PIP in " (Utilities|String|ToString(Float) _output_get) " Seconds"))
      (Utilities|FlowControl|Delay _output_get)
      (Variables|Default|SetCurrentLiveStreamInMatch (Variables|/Script/FortniteGame.FortRuntimeOptions|GetbEnableLiveStreamInMatch _returnvalue))
      (Variables|/Script/FortniteGame.FortRuntimeOptions|SetbEnableLiveStreamInMatch true _returnvalue)
      (Variables|Default|SetCurrentPiPLiveByDefault (Variables|/Script/FortniteGame.FortRuntimeOptions|GetbShowLiveStreamInMatchByDefault _returnvalue))
      (Variables|/Script/FortniteGame.FortRuntimeOptions|SetbShowLiveStreamInMatchByDefault true _returnvalue)
      (Variables|Default|SetCurrentHiddenPiPHF (Variables|/Script/FortniteGame.FortRuntimeOptions|GetbHiddenButEnabledLiveStreamInMatch _returnvalue))
      (Variables|/Script/FortniteGame.FortRuntimeOptions|SetbHiddenButEnabledLiveStreamInMatch true _returnvalue)
      (Variables|Default|SetCurrentPiPCounter _shouldshowlivestreampipinmatchcounter)
      (bind _output_get_2 (Variables|Default|SetCounterBuffer (Math|Random|RandomInteger64InRange 0 100)))
      (if (Math|Integer64|Equal(Integer64) _output_get_2 _shouldshowlivestreampipinmatchcounter)
        _)
      (Variables|/Script/FortniteGame.FortRuntimeOptions|SetShouldShowLiveStreamPiPInMatchCounter _output_get_2 _returnvalue))))

(event EventEndPlay
  (params (EndPlayReason :type EEndPlayReason%20Enum :default "Destroyed"))
  (bind _returnvalue (Options|GetRuntimeOptions))
  ;[1 1 1] End Play
  (switch Utilities|FlowControl|Switch|SwitchonEMeshNetworkNodeType (MeshNetworkSubsystem|GetMeshNetworkNodeType (GameInstanceSubsystems|GetMeshNetworkSubsystem))
    (:Client
      (Variables|/Script/FortniteGame.FortRuntimeOptions|SetbEnableLiveStreamInMatch (Variables|Default|GetCurrentLiveStreamInMatch) _returnvalue)
      (Variables|/Script/FortniteGame.FortRuntimeOptions|SetbShowLiveStreamInMatchByDefault (Variables|Default|GetCurrentPiPLiveByDefault) _returnvalue)
      (Variables|/Script/FortniteGame.FortRuntimeOptions|SetShouldShowLiveStreamPiPInMatchCounter (Variables|Default|GetCurrentPiPCounter) _returnvalue)
      (Variables|/Script/FortniteGame.FortRuntimeOptions|SetbHiddenButEnabledLiveStreamInMatch (Variables|Default|GetCurrentHiddenPiPHF) _returnvalue))))

(fn ConstructionScript)
