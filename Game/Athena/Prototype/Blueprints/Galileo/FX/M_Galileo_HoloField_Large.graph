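# Additive, unlit, two-sided material for the large Galileo hologram field, usable on mesh particles.
# Color comes from a curve atlas sampled by Fresnel and vertical gradient, plus a tint and timed flash.
# Visibility combines time-of-day brightness, inner-face scaling and angle, depth and near-camera fades, simplified on mobile.
# A vertex-offset term widens the top, and camera distance beyond a threshold adds height and adjusts the fades.
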Material /Game/Athena/Prototype/Blueprints/Galileo/FX/M_Galileo_HoloField_Large
graph {
 properties {
  bUsedWithMeshParticles=true
  blendMode=BLEND_Additive
  pixelDepthOffsetMode=PDOM_Legacy
  shadingModel=MSM_Unlit
  twoSided=true
  usedShadingModels=Unlit
 }
 nodes [
  n1 = Subtract(A=Time() {period=5}, B=ScalarParameter() {parameterName=FlashStartTime})
  n2 = ScalarParameter() {defaultValue=0.75, group=Timestamp, parameterName=FlashDuration}
  n3 = MaterialFunctionCall!(In=Clamp(None=Divide(A=@n1, B=@n2))) {materialFunction=/Game/Effects/Fort_Effects/Materials/_Functions/MF_Smoothstep_EaseOut}
  n4 = ScalarParameter() {defaultValue=4, parameterName=FlashBrightness}
  n5 = ScalarParameter() {defaultValue=5, parameterName=Brightness}
  n6 = ComponentMask(None=TextureCoordinate()) {g=true}
  n7 = Fresnel(
   ExponentIn=LinearInterpolate(
    A=ScalarParameter() {defaultValue=2, parameterName=FresnelExpTop}
    B=ScalarParameter() {defaultValue=0.7, parameterName=FresnelExpBase}
    Alpha=@n6
   )
   BaseReflectFractionIn=ScalarParameter() {defaultValue=0.3, parameterName=FresnelBaseBrightness}
  ) {baseReflectFraction=0.2, exponent=1.5}
  n8 = Constant() {r=0.6}
  n9 = FeatureLevelSwitch(Default=@n7, ES2=@n8, ES3_1=@n8, SM4_REMOVED=@n7)
  n10 = ComponentMask(None=TextureCoordinate()) {g=true}
  n11 = Power(Base=@n10, Exp=ScalarParameter() {defaultValue=10, parameterName=GradientPower}) {constExponent=3}
  n12 = CurveAtlasRowParameter(CurveTime=Saturate(None=Multiply(A=@n9, B=@n11))) {
   atlas=&AthenaCurveAtlas01
   curve=/Game/Athena/LinearColorCurves/Curve_Galileo_Hologram
   defaultValue=171
   parameterName=Curve
  }
  Multiply(
   A=Add(A=Multiply(A=@n3.Result, B=@n4))
   B=Add(A=Multiply(A=@n5, B=@n12), B=VectorParameter() {parameterName=ColorAdd, defaultValue=<rgba 0.1367 0.5081 1.25 0>}.RGB)
  )
  n13 = ScalarParameter() {defaultValue=1.2, parameterName=DayBrightness}
  n14 = Multiply(A=Constant() {r=1.2}, B=@n13)
  n15 = ScalarParameter() {defaultValue=0.5, parameterName=NightBrightness}
  n16 = ScalarParameter() {defaultValue=0.25, parameterName=InnerBrightness}
  n17 = ConstantBiasScale(None=TwoSidedSign())
  n18 = Constant3Vector() {constant=<rgba 0 0 1 0>}
  n19 = Distance(A=ActorPositionWS(), B=CameraPositionWS())
  n20 = Saturate(None=Divide(A=Add(A=@n19) {constB=-50000}, B=ScalarParameter() {defaultValue=70000, parameterName=FadeRange}))
  n21 = Add(
   A=LinearInterpolate(
    A=ScalarParameter() {defaultValue=5, parameterName=AngleFadeExpTop}
    B=ScalarParameter() {defaultValue=0.75, parameterName=AngleFadeExpBase}
    Alpha=@n6
   )
   B=@n20
  )
  n22 = Multiply(A=VertexNormalWS(), B=TwoSidedSign())
  n23 = Multiply(
   A=Power(
    Base=MaterialFunctionCall!(Axis=@n18, Exponent=@n21, Normal=@n22) {materialFunction=/Game/Effects/Fort_Effects/Materials/_Functions/MF_AxisAlignedFrensnel}.Fresnel
    Exp=ScalarParameter() {defaultValue=2.25, parameterName=AngleFadePower}
   )
   B=ScalarParameter() {defaultValue=3, parameterName=AngleFadeMult}
  ) {constB=5}
  n24 = ScalarParameter() {defaultValue=500, parameterName=MaxDepthFade}
  n25 = VertexColor()
  n26 = Multiply(
   A=DepthFade(Opacity=Saturate(None=@n23), FadeDistance=Multiply(A=@n24, B=@n25.G))
   B=MaterialFunctionCall(Distance=ScalarParameter() {defaultValue=500, parameterName=NearFadeDistance}) {materialFunction=&Near_Camera_Fade}.Result
  )
  n27 = Constant() {r=1}
  n28 = Multiply(A=Multiply(A=@n9, B=@n10), B=FeatureLevelSwitch(Default=@n26, ES2=@n27, ES3_1=@n27, SM4_REMOVED=@n26))
  n29 = Add(A=Constant() {r=0.4}, B=@n20) {constB=0.4}
  Multiply(
   A=Multiply(
    A=MaterialFunctionCall(Color_Morning=@n13, Color_Day=@n14, Color_Evening=@n13, Color_Night=@n15) {materialFunction=&Color_Time_of_Day}.Result
    B=LinearInterpolate(A=@n16, Alpha=@n17) {constA=0.25}
   )
   B=Multiply(A=Multiply(A=@n28, B=@n29) {constB=0.4}, B=ScalarParameter() {defaultValue=1, parameterName=OpacityMult})
  ) {constA=1}
  n30 = TextureCoordinate()
  n31 = Subtract(A=WorldPosition().XYZ, B=ActorPositionWS())
  n32 = Multiply(A=@n31, B=Constant3Vector() {constant=<rgba 1 1 0 0>})
  n33 = ScalarParameter() {defaultValue=1.5, parameterName=TopWiden}
  n34 = Multiply(A=@n31, B=Constant3Vector() {constant=<rgba 0 0 1 0>})
  n35 = Multiply(A=ScalarParameter() {defaultValue=2, parameterName=TopMaxHeightAdd}, B=@n20)
  Multiply(A=OneMinus(None=ComponentMask(None=@n30) {g=true}), B=Add(A=Multiply(A=@n32, B=@n33), B=Multiply(A=@n34, B=@n35)))
 ]
}
