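# Translucent unlit Galileo hologram material that displays a SlateUI texture with animated noise, flicker and scanlines.
# Panning world-aligned noise, gated by layered sines, modulates screen-texture opacity, curve-atlas color and positional jitter; tinted UI adds emissive and masks opacity.
# TurnOn fades opacity while a vertical sphere-mask pulse sweeps through.
# World position offset also supports camera-facing rotation and hiding by displacement, while Blackout forces opaque black without offset.
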Material /Game/Athena/Prototype/Blueprints/Galileo/Materials/M_Galileo_Hologram_UIText
graph {
 properties {
  bEnableResponsiveAA=true
  bUsedWithMeshDeformer=true
  bUsedWithNanite=true
  bUsedWithSkeletalMesh=true
  blendMode=BLEND_Translucent
  pixelDepthOffsetMode=PDOM_Legacy
  shadingModel=MSM_Unlit
  twoSided=true
  usedShadingModels=Unlit
 }
 nodes [
  n1 = ScalarParameter() {parameterName=Blackout}
  n2 = TextureSampleParameter2D() {group=Texture, parameterName=2DHologramTexture, texture=&/T_Galileo_Screen}
  n3 = VectorParameter() {group=Noise, parameterName=NoiseTileMult, defaultValue=<rgba 2 2 0.1 1>}
  n4 = ScalarParameter() {defaultValue=1500, group=Noise, parameterName=NoiseSize}
  n5 = WorldPosition()
  n6 = Multiply(A=@n4, B=Multiply(A=Time(), B=VectorParameter() {group=Noise, parameterName=NoisePanSpeed, defaultValue=<rgba -0.1 0.09 -0.15 0>}.RGB))
  n7 = Power(Base=MaterialFunctionCall(
    TextureObject=TextureObject() {samplerType=SAMPLERTYPE_LinearColor, texture=&T_NoisePack_LinearColor}
    TextureSize=Multiply(A=@n3.RGB, B=@n4)
    WorldPosition=Add(A=@n5.XYZ, B=@n6)
   ) {materialFunction=&WorldAlignedTexture}.XY Texture) {constExponent=4}
  n8 = ScalarParameter() {defaultValue=2, group=Noise, parameterName=NoiseMaskMult}
  n9 = Constant() {r=0.3}
  n10 = ScalarParameter() {defaultValue=0.15, group=Noise, parameterName=NoiseAdd}
  n11 = Time()
  n12 = Multiply(A=ScalarParameter() {defaultValue=1, parameterName=FlickerSpeed}, B=@n11)
  n13 = AppendVector(A=AppendVector(A=Sine(None=@n12) {period=1.6}, B=Sine(None=@n12) {period=1.42}), B=Sine(None=@n12) {period=1.51})
  n14 = ScalarParameter() {defaultValue=0.5, group=Noise, parameterName=NoiseFrequency}
  n15 = Add(A=Sine(None=@n12) {period=2.1}, B=Multiply(B=@n14) {constA=0.25})
  n16 = Add(A=Sine(None=@n12) {period=0.72}, B=Multiply(B=@n14) {constA=0.75}) {constB=0.75}
  n17 = Ceil(None=Add(A=Multiply(B=@n14) {constA=0.9, constB=0.9}, B=Sine(None=@n12) {period=1.25}) {constB=0.9})
  n18 = Saturate(None=Multiply(A=Multiply(A=@n15, B=@n16), B=Multiply(A=@n16, B=@n17)))
  n19 = LinearInterpolate(Alpha=ConstantBiasScale(None=Sine(None=@n11) {period=0.075})) {constA=0.7}
  n20 = Multiply(A=Multiply(A=Ceil(None=@n13), B=Multiply(A=@n18, B=@n19)), B=ScalarParameter() {defaultValue=2, group=Noise, parameterName=NoiseMult})
  n21 = ScalarParameter() {defaultValue=1, parameterName=TurnOn}
  n22 = OneMinus(None=@n21)
  n23 = Multiply(A=QualitySwitch(Default=Multiply(A=@n7, B=@n8) {constB=20}, Low=@n9), B=Add(A=Add(A=@n10, B=@n20), B=@n22) {constB=0})
  n24 = WorldPosition()
  n25 = ComponentMask(None=ActorPositionWS()) {b=true}
  n26 = Add(A=@n25, B=ScalarParameter() {parameterName=PulseStartOffset})
  n27 = Add(A=@n25, B=ScalarParameter() {defaultValue=1600, parameterName=PulseEndOffset}) {constB=1600}
  n28 = ScalarParameter() {defaultValue=300, parameterName=PulseRadius}
  n29 = Multiply(A=@n28) {constB=0.1}
  n30 = Multiply(
   A=SphereMask(
    A=ComponentMask(None=@n24.XYZ) {b=true}
    B=LinearInterpolate(A=@n26, B=@n27, Alpha=@n21) {constA=-0.25, constB=1.2}
    Radius=LinearInterpolate(A=@n28, B=@n29, Alpha=@n21) {constA=500, constB=100}
   ) {hardnessPercent=0}
   B=@n22
  )
  n31 = Desaturation(None=Add(A=Saturate(None=@n23), B=Multiply(A=@n30) {constB=3}))
  n32 = MaterialFunctionCall(In=@n31, Contrast=ScalarParameter() {defaultValue=0.2, group=Noise, parameterName=NoiseContrast}) {materialFunction=&CheapContrast}
  n33 = Multiply(A=ScalarParameter() {defaultValue=-0.5, group=Noise, parameterName=NoisePowerAdd}, B=@n32.Result)
  n34 = Add(A=Power(Base=@n2.G, Exp=Add(A=@n33)), B=Multiply(A=@n30) {constB=0.5})
  n35 = Multiply(A=@n34, B=ScalarParameter() {parameterName=BackgroundMult})
  n36 = VectorParameter() {parameterName=SlateScale, defaultValue=<rgba 1 1 0.5 0.5>}
  n37 = Constant() {r=1}
  n38 = TextureSampleParameter2D(UVs=MaterialFunctionCall!(
    Texture Scale=AppendVector(A=@n36.R, B=@n36.R)
    UVs
    Offset XY Texturesize Z=MaterialFunctionCall(R=@n36.B, G=@n36.B, B=@n37) {materialFunction=&AppendMany}.RGB
   ) {materialFunction=/Engine/Functions/Engine_MaterialFunctions01/Texturing/ScaleUVsAroundPoint}.UVs) {parameterName=SlateUI, samplerSource=SSM_Clamp_WorldGroupSettings, texture=&Black}
  n39 = ScalarParameter() {defaultValue=-1.25, group=Noise, parameterName=NoiseBrightness}
  n40 = Saturate(None=@n32.Result)
  n41 = Add(A=@n34, B=Multiply(A=@n2.G, B=Multiply(A=@n39, B=@n40) {constB=0.1}))
  n42 = ScalarParameter() {defaultValue=0.4, parameterName=MinScanlineOpacity}
  n43 = Multiply(A=@n32.Result) {constB=1.5}
  n44 = ScalarParameter() {defaultValue=0.1, parameterName=ScanlineTile}
  n45 = LinearInterpolate(Alpha=Ceil(None=@n40)) {constA=1, constB=1.25}
  n46 = WorldPosition()
  n47 = Sine(None=Multiply(A=Multiply(A=@n44, B=@n45), B=ComponentMask(None=@n46.XYZ) {b=true}) {constA=1})
  n48 = LinearInterpolate(A=Clamp(None=Subtract(A=@n42, B=@n43)) {minDefault=-0.25}, B=Add(A=@n43), Alpha=Power(Base=ConstantBiasScale(None=@n47)) {constExponent=1.5}) {constA=0.5}
  n49 = Multiply(A=Saturate(None=Multiply(A=@n41, B=@n48)), B=ScalarParameter() {defaultValue=1, parameterName=OpacityMult})
  Add(A=@n1, B=Multiply(A=LinearInterpolate(A=@n35, Alpha=@n38.A), B=Multiply(A=@n49, B=@n21)))
  n50 = Constant() {r=1}
  n51 = ScalarParameter() {parameterName=EmissiveAdd}
  n52 = CurveAtlasRowParameter(CurveTime=Saturate(None=@n34)) {
   atlas=&AthenaCurveAtlas01
   curve=/Game/Athena/LinearColorCurves/Curve_Galileo_Hologram
   defaultValue=171
   parameterName=ColorCurve
  }
  n53 = Multiply(A=@n52, B=ScalarParameter() {defaultValue=2, parameterName=EmissiveMult}) {constB=2}
  n54 = Multiply(
   A=Multiply(
    A=VectorParameter() {parameterName=UIBrightness, defaultValue=<rgba 0.1681 1.9599 3 0>}.RGB
    B=ScalarParameter() {defaultValue=1, parameterName=UIMultiplier}
   )
   B=@n38.RGB
  )
  LinearInterpolate(A=Multiply(A=Add(A=@n50, B=@n51), B=Add(A=@n53, B=@n54)), Alpha=@n1) {constB=0}
  n55 = WorldPosition()
  n56 = MaterialFunctionCall() {materialFunction=&ObjectPivotPoint}
  n57 = Multiply(A=Normalize(VectorInput=Subtract(A=@n55.XYZ, B=@n56.Object Pivot Location)), B=Constant3Vector() {constant=<rgba 1 1 0 0>})
  n58 = ScalarParameter() {defaultValue=20, group=Noise, parameterName=WPOStrength}
  n59 = Subtract(A=QualitySwitch(Default=@n7, Low=@n9)) {constB=0.5}
  n60 = Add(
   A=Multiply(A=@n31, B=Multiply(A=@n57, B=@n58)) {constB=0}
   B=Multiply(A=Multiply(A=@n58, B=@n59), B=ScalarParameter() {parameterName=Distort})
  )
  n61 = CameraPositionWS()
  n62 = ActorPositionWS()
  n63 = Normalize(VectorInput=Multiply(A=Constant3Vector() {constant=<rgba 1 1 0 0>}, B=Subtract(A=@n61, B=@n62)))
  n64 = Constant3Vector() {constant=<rgba 0 1 0 0>}
  n65 = Divide(A=MaterialFunctionCall!(Normalized V1=@n63, Normalized V2=Normalize(VectorInput=@n64)) {materialFunction=/Engine/Functions/Engine_MaterialFunctions02/Math/AngleBetweenVectors}.Angle In Degrees) {constB=360}
  n66 = Normalize(VectorInput=Constant3Vector() {constant=<rgba 1 0 0 0>})
  n67 = Multiply(A=Sign(None=DotProduct(A=@n63, B=@n66))) {constB=-1}
  n68 = LinearInterpolate(
   B=RotateAboutAxis(
    NormalizedRotationAxis=Constant3Vector() {constant=<rgba 0 0 1 0>}
    RotationAngle=Multiply(A=@n65, B=@n67)
    PivotPoint=@n62
    Position=WorldPosition().XYZ
   )
   Alpha=ScalarParameter() {parameterName=FaceCamera}
  )
  n69 = VectorParameter() {parameterName=HideOffset, defaultValue=<rgba 0 0 -50000 0>}
  n70 = ScalarParameter() {defaultValue=1, parameterName=Visible}
  LinearInterpolate(A=Add(A=Add(A=@n60, B=@n68), B=LinearInterpolate(A=@n69.RGB, Alpha=@n70) {constB=0}), Alpha=@n1) {constB=0}
 ]
}
