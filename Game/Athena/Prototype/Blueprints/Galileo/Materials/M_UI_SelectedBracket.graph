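# Translucent UI material that procedurally draws a selection bracket, judging by its name, without sampling any textures.
# An inverted, center-scaled UV mask sized by OutlineSize forms the outer border.
# A second center-scaled mask, narrowed horizontally by InnerCutaway at full height, is subtracted to cut out the middle of that border, leaving bracket shapes on the left and right.
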Material /Game/Athena/Prototype/Blueprints/Galileo/Materials/M_UI_SelectedBracket
graph {
 properties {blendMode=BLEND_Translucent, materialDomain=MD_UI, pixelDepthOffsetMode=PDOM_Legacy, usedShadingModels=Default Lit}
 nodes [
  Constant() {r=1}
  n1 = ScalarParameter() {defaultValue=0.9, parameterName=OutlineSize}
  n2 = ScalarParameter() {defaultValue=0.7, parameterName=InnerCutaway}
  n3 = Constant() {r=1}
  Subtract(
   A=OneMinus(None=MaterialFunctionCall(Texture Scale=@n1) {materialFunction=&ScaleUVsByCenter}.0-1 mask)
   B=MaterialFunctionCall(Texture Scale=AppendVector(A=@n2, B=@n3)) {materialFunction=&ScaleUVsByCenter}.0-1 mask
  )
 ]
}
