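# M_UI_SquarePulse is a translucent UI material prototype that builds a rotatable square outline mask.
# It rotates centered UVs by Rotation, then subtracts a mask scaled by Scale minus OutlineWidth from one scaled by Scale.
# The aspect correction divides pixel-size X by itself, so it has no stretching effect.
# The graph has no time-driven pulse, and the final mask shows no connection to a material output.
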Material /Game/Athena/Prototype/Blueprints/Galileo/Materials/M_UI_SquarePulse
graph {
 properties {blendMode=BLEND_Translucent, materialDomain=MD_UI, pixelDepthOffsetMode=PDOM_Legacy, usedShadingModels=Default Lit}
 nodes [
  Constant() {r=1}
  n1 = TextureCoordinate()
  n2 = MaterialFunctionCall(Float2=MaterialFunctionCall() {materialFunction=&GetUserInterfaceUV}.Pixel Size) {materialFunction=&BreakOutFloat2Components}
  n3 = AppendVector(A=Constant() {r=1}, B=Divide(A=@n2.R, B=@n2.R))
  Multiply(A=@n1, B=@n3)
  n4 = MaterialFunctionCall(
   UVs=MaterialFunctionCall(UVs=@n1, Texture Scale=@n3) {materialFunction=&ScaleUVsByCenter}.UVs
   "Rotation Angle (0-1)"=ScalarParameter() {parameterName=Rotation}
  ) {materialFunction=&CustomRotator}
  n5 = ScalarParameter() {defaultValue=1, parameterName=Scale}
  n6 = ScalarParameter() {defaultValue=0.1, parameterName=OutlineWidth, sliderMax=1}
  Subtract(
   A=MaterialFunctionCall(UVs=@n4.Rotated Values, Texture Scale=@n5) {materialFunction=&ScaleUVsByCenter}.0-1 mask
   B=MaterialFunctionCall(UVs=@n4.Rotated Values, Texture Scale=Subtract(A=@n5, B=@n6)) {materialFunction=&ScaleUVsByCenter}.0-1 mask
  )
 ]
}
